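# Niagara system for the Xenosplash car boost: green alien-liquid sprites and ribbon trails from the vehicle's boost sockets, recolored when painted.
# The GPU Sprites emitter merges three layers through spawn groups, with per-group values and visibility tags for three renderers.
# Trail001 draws dual-renderer ribbons per socket with extra color bindings.
# A frontend-preview flag adjusts motion, sprite rates and sizes; several variant emitters are disabled.
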NiagaraSystem /VehicleCosmetics/Boosts/Boost_Xenosplash/FX/NS_Boost_AlienLiquid_Optimized
dependencies {
 usedDataInterfaces [
  &NiagaraDataInterfaceCurve
  &NiagaraDataInterfaceSpriteRendererInfo
  &NiagaraDataInterfaceSkeletalMesh
 ]
 usedDynamicInputs []
 usedModules [
  &self:SystemSpawnScript.SystemScriptSource.SystemScriptGraph.NiagaraNodeAssignment_0.SetVariables_#1
  &SystemState
 ]
}
emitters {
 Sprites {
  simTarget=GPUComputeSim
  emitterSpawnScript [/SetVariables(Emitter.AcidDistortion:NiagaraFloat=&RandomRangeFloat(Minimum=0, Maximum=0.5)) {isSetParameters=true, module=SetVariables_#2}]
  emitterUpdateScript [
   &EmitterState()
   &VC_FilteredSocketCount()
   &SpawnRate(
    SpawnRate=&MakeCustomFloatFromBool(Bool=$User.bIsFrontendPreview, True Float=20, False Float=40)
    Use Spawn Probability=false
   )
   &SpawnRate(
    SpawnGroup=1
    SpawnRate=&MakeCustomFloatFromBool(Bool=$User.bIsFrontendPreview, True Float=50, False Float=80)
    Use Spawn Probability=false
   ) {module=SpawnRate001}
   &SpawnRate(SpawnGroup=2, SpawnRate=45, Use Spawn Probability=false) {module=SpawnRate002}
  ]
  particleSpawnScript [
   &VC_PaintedVariantsOverride() {enabled=false}
   /SetVariables(
    Particles.Base Color:LinearColor=&MakeCustomLinearColorFromBool(Bool=$User.bIsPainted, True Color=$User.PaintedColor, False Color=<rgba 0.2 0.8 0.05 1>)
    Particles.VisibilityTag=&SpawnGroup()
   ) {isSetParameters=true, module=SetVariables_#3}
   &InitializeParticle(
    Color Mode=::NewEnumerator0
    Lifetime Max=/VehicleCosmetics/FX/Niagara/Modules/MergeEmitters/VC_FloatBySpawnGroup(
     SpawnGroup or ExecIndex:NiagaraInt32=0
     CustomInt:NiagaraInt32=0
     MaximumSpawnGroups:NiagaraInt32=3
     Float SG 0:NiagaraFloat=1
     Float SG 1:NiagaraFloat=0.6
     Float SG 2:NiagaraFloat=1.5
     Float SG 3:NiagaraFloat=0
     Float SG 4:NiagaraFloat=0
     Float SG 5:NiagaraFloat=0
    )
    Lifetime Min:NiagaraFloat=&VC_FloatBySpawnGroup(
     SpawnGroup or ExecIndex:NiagaraInt32=0
     CustomInt:NiagaraInt32=0
     MaximumSpawnGroups:NiagaraInt32=3
     Float SG 0:NiagaraFloat=0.5
     Float SG 1:NiagaraFloat=0.5
     Float SG 2:NiagaraFloat=1.5
     Float SG 3:NiagaraFloat=0
     Float SG 4:NiagaraFloat=0
     Float SG 5:NiagaraFloat=0
    )
    Lifetime Mode=::NewEnumerator1
    Sprite Rotation Mode=::NewEnumerator2
    Sprite Size Mode=::NewEnumerator3
    Uniform Sprite Size=&VC_FloatBySpawnGroup(
     SpawnGroup or ExecIndex:NiagaraInt32=0
     CustomInt:NiagaraInt32=&SpawnGroup()
     MaximumSpawnGroups:NiagaraInt32=3
     Float SG 0:NiagaraFloat=60
     Float SG 1:NiagaraFloat=&RandomRangeFloat(Minimum=90, Maximum=150)
     Float SG 2:NiagaraFloat=&RandomRangeFloat(
      Minimum=&MakeCustomFloatFromBool(Bool=$User.bIsFrontendPreview, True Float=3, False Float=4)
      Maximum=&MakeCustomFloatFromBool(Bool=$User.bIsFrontendPreview, True Float=6, False Float=8)
     )
     Float SG 3:NiagaraFloat=0
     Float SG 4:NiagaraFloat=0
     Float SG 5:NiagaraFloat=0
    )
    UsePositionOffset=false
    Write Material Random=true
   )
   &PartitionParticles(Number Of Partitions=$Emitter.PortCount, Partition Mode:ENiagara_PartitionMode=::NewEnumerator1)
   &SpriteFacingAndAlignment(
    Facing Coordinate Space=::Local
    Sprite Alignment=&MakeCustomVectorFromBool(Bool=$User.bIsFrontendPreview, True Vector=<0 1 0>, False Vector=$Engine.Owner.Velocity)
    WriteAlignment=false
    WriteFacing=true
   )
   &SkeletalMeshLocation(
    Bone / Socket Index=$Particles.PartitionParticles.Partition
    Bone / Socket Sampling Mode=::NewEnumerator5
    MaskSpawnGroup=false
    Mesh Sampling Type=::NewEnumerator6
    Sampled Position Offset=&AddVectorToPosition(
     Position=&MakeVector(
      X=&VC_FloatBySpawnGroup(
       SpawnGroup or ExecIndex:NiagaraInt32=0
       CustomInt:NiagaraInt32=0
       MaximumSpawnGroups:NiagaraInt32=3
       Float SG 0:NiagaraFloat=-10
       Float SG 1:NiagaraFloat=-50
       Float SG 2:NiagaraFloat=-50
       Float SG 3:NiagaraFloat=0
       Float SG 4:NiagaraFloat=0
       Float SG 5:NiagaraFloat=0
      )
      Y=0
      Z=0
     )
     Vector=$User.DTCarOffset
    )
    Skeletal Mesh=$System.SkeletalMesh
   )
   &AddVelocity(Velocity=/Niagara/DynamicInputs/TypeConversions/MakeCustomPositionFromBool(
     Bool:NiagaraBool=$User.bIsFrontendPreview
     True Position:NiagaraPosition=&MakeVector(
      X=&VC_FloatBySpawnGroup(
       SpawnGroup or ExecIndex:NiagaraInt32=0
       CustomInt:NiagaraInt32=0
       MaximumSpawnGroups:NiagaraInt32=3
       Float SG 0:NiagaraFloat=-950
       Float SG 1:NiagaraFloat=-850
       Float SG 2:NiagaraFloat=-850
       Float SG 3:NiagaraFloat=0
       Float SG 4:NiagaraFloat=0
       Float SG 5:NiagaraFloat=0
      )
      Y=&MaskFloatBySpawnGroup(Masked Float=&RandomRangeFloat(Minimum=0, Maximum=0), Passthrough Float=0, Spawn Group Mask=2)
      Z=&MaskFloatBySpawnGroup(Masked Float=&RandomRangeFloat(Minimum=0, Maximum=0), Passthrough Float=0, Spawn Group Mask=2)
     )
     False Position:NiagaraPosition=&MakeVector(
      X=&VC_FloatBySpawnGroup(
       SpawnGroup or ExecIndex:NiagaraInt32=0
       CustomInt:NiagaraInt32=0
       MaximumSpawnGroups:NiagaraInt32=3
       Float SG 0:NiagaraFloat=0
       Float SG 1:NiagaraFloat=-212.5
       Float SG 2:NiagaraFloat=-212.5
       Float SG 3:NiagaraFloat=0
       Float SG 4:NiagaraFloat=0
       Float SG 5:NiagaraFloat=0
      )
      Y=0
      Z=0
     )
    ))
   &Color(Color=/VehicleCosmetics/FX/Niagara/Modules/MergeEmitters/VC_Dynamic_ColorBySpawnGroup(
     Color SG 0:LinearColor=$Particles.Base Color
     MaximumSpawnGroups:NiagaraInt32=3
     Color SG 1:LinearColor=&MakeCustomLinearColorFromBool(Bool=$User.bIsPainted, True Color=$User.PaintedColor, False Color=<rgba 0.5 1 0.01 1>)
     Color SG 2:LinearColor=$Particles.Base Color
     Color SG 3:LinearColor=<rgba 1 1 1 1>
     Color SG 4:LinearColor=<rgba 1 1 1 1>
    ))
   &ShapeLocation(
    MaskSpawnGroup=false
    Override Seed=false
    Sphere Radius=&MaskFloatBySpawnGroup(Masked Float=50, Passthrough Float=0, Spawn Group Mask=2)
    UseOwnerScale=true
   )
  ]
  particleUpdateScript [
   &ParticleState()
   &ScaleSpriteSize(
    Scale Sprite Size Mode=::NewEnumerator0
    Uniform Scale Factor=&VC_FloatBySpawnGroup(
     SpawnGroup or ExecIndex:NiagaraInt32=0
     CustomInt:NiagaraInt32=0
     MaximumSpawnGroups:NiagaraInt32=3
     Float SG 0:NiagaraFloat=&FloatFromCurve(FloatCurve=DataInterface() {propertyValues {Curve {keys [{arriveTangent=2, interpMode=RCIM_Cubic, time=0.3, value=1.15}, <at 1 2 RCIM_Cubic>]}}})
     Float SG 1:NiagaraFloat=1
     Float SG 2:NiagaraFloat=1
     Float SG 3:NiagaraFloat=0
     Float SG 4:NiagaraFloat=0
     Float SG 5:NiagaraFloat=0
    )
   )
   &SpriteRotationRate(Rotation Rate=&MaskFloatBySpawnGroup(Masked Float=&RandomRangeFloat(Minimum=-0.2, Maximum=0.2), Passthrough Float=0, Spawn Group Mask=1))
   &ScaleVelocity(
    Coordinate Space:ENiagaraCoordinateSpace=::Simulation
    Velocity Scale=&VectorFromFloat(Value=&VC_FloatBySpawnGroup(
      SpawnGroup or ExecIndex:NiagaraInt32=0
      CustomInt:NiagaraInt32=0
      MaximumSpawnGroups:NiagaraInt32=3
      Float SG 0:NiagaraFloat=&MakeCustomFloatFromBool(
       Bool=$User.bIsFrontendPreview
       True Float=&FloatFromCurve(FloatCurve=DataInterface() {propertyValues {Curve {keys [<at 0 0.8>,<at 0.5 0.15>]}}})
       False Float=&FloatFromCurve(FloatCurve=DataInterface() {propertyValues {Curve {keys [<at 0 0.25>,<at 1 0>]}}})
      )
      Float SG 1:NiagaraFloat=&FloatFromCurve(FloatCurve=DataInterface() {propertyValues {Curve {keys [<at 0 1 RCIM_Cubic>,<at 1 1.25 RCIM_Cubic>]}}})
      Float SG 2:NiagaraFloat=&FloatFromCurve(FloatCurve=DataInterface() {propertyValues {Curve {keys [<at 0 0.65 RCIM_Cubic>,<at 1 0.2 RCIM_Cubic>]}}})
      Float SG 3:NiagaraFloat=0
      Float SG 4:NiagaraFloat=0
      Float SG 5:NiagaraFloat=0
     ))
   )
   &CurlNoiseForce(
    Bias Noise=false
    Noise Frequency=&MaskFloatBySpawnGroup(Masked Float=10, Passthrough Float=0, Spawn Group Mask=2)
    Noise Strength=&MaskFloatBySpawnGroup(Masked Float=1500, Passthrough Float=0, Spawn Group Mask=2)
    Pan Noise=false
   ) {enabled=false}
   &SolveForcesAndVelocity(Clamp Velocity=false, Limit Acceleration=false)
   &ScaleColor(
    Scale Alpha=&VC_FloatBySpawnGroup(
     SpawnGroup or ExecIndex:NiagaraInt32=0
     CustomInt:NiagaraInt32=0
     MaximumSpawnGroups:NiagaraInt32=3
     Float SG 0:NiagaraFloat=&FloatFromCurve(FloatCurve=DataInterface() {propertyValues {Curve {keys [<at 0 1 RCIM_Cubic>,<at 1 0 RCIM_Cubic>]}}})
     Float SG 1:NiagaraFloat=&FloatFromCurve(FloatCurve=DataInterface() {propertyValues {Curve {keys [<at 0 1 RCIM_Cubic>,<at 1.0011 0.002 RCIM_Cubic>]}}})
     Float SG 2:NiagaraFloat=&FloatFromCurve(FloatCurve=DataInterface() {propertyValues {Curve {keys [<at -0.0046 1.0113 RCIM_Cubic>,<at 1.0011 0.002 RCIM_Cubic>]}}})
     Float SG 3:NiagaraFloat=0
     Float SG 4:NiagaraFloat=0
     Float SG 5:NiagaraFloat=0
    )
    Scale RGB=&VectorFromFloat(Value=&VC_FloatBySpawnGroup(
      SpawnGroup or ExecIndex:NiagaraInt32=0
      CustomInt:NiagaraInt32=0
      MaximumSpawnGroups:NiagaraInt32=3
      Float SG 0:NiagaraFloat=&Multiply_Float(A=2, B=$User.PaintedScalar)
      Float SG 1:NiagaraFloat=$User.PaintedScalar
      Float SG 2:NiagaraFloat=&Multiply_Float(A=6, B=$User.PaintedScalar)
      Float SG 3:NiagaraFloat=0
      Float SG 4:NiagaraFloat=0
      Float SG 5:NiagaraFloat=0
     ))
    ScaleA=true
    ScaleRGB=true
   )
   &SubUVAnimation(
    End Frame Range Override=16
    Output Alternate Renderer Binding=true
    "Sprite Renderer "=DataInterface() {propertyValues {SpriteRenderer=&self:Sprites_0.NiagaraSpriteRendererProperties_0}}
    UseEndFrame=true
    UseStartFrame=false
   )
   &JitterPosition(Jitter Amount:NiagaraFloat=&VC_FloatBySpawnGroup(
     SpawnGroup or ExecIndex:NiagaraInt32=0
     CustomInt:NiagaraInt32=0
     MaximumSpawnGroups:NiagaraInt32=3
     Float SG 0:NiagaraFloat=0
     Float SG 1:NiagaraFloat=&RandomRangeFloat(Minimum=-2, Maximum=2)
     Float SG 2:NiagaraFloat=&RandomRangeFloat(Minimum=-2, Maximum=2)
     Float SG 3:NiagaraFloat=0
     Float SG 4:NiagaraFloat=0
     Float SG 5:NiagaraFloat=0
    ))
  ]
  rendererClasses [&NiagaraSpriteRendererProperties]
  renderers [
   {rendererClass=&NiagaraSpriteRendererProperties}
   {rendererClass=&NiagaraSpriteRendererProperties, rendererIndex=1}
   {rendererClass=&NiagaraSpriteRendererProperties, rendererIndex=2}
  ]
 }
 Sprites001 {
  bEnabled=false
  simTarget=GPUComputeSim
  emitterSpawnScript [/SetVariables(Emitter.AcidDistortion:NiagaraFloat=&RandomRangeFloat(Minimum=0, Maximum=0.5)) {isSetParameters=true, module=SetVariables_#2}]
  emitterUpdateScript [
   &EmitterState()
   &VC_FilteredSocketCount()
   &SpawnRate(SpawnRate=20, Use Spawn Probability=false)
   &SpawnRate(
    SpawnGroup=1
    SpawnRate=&MakeCustomFloatFromBool(Bool=$User.bIsFrontendPreview, True Float=50, False Float=80)
    Use Spawn Probability=false
   ) {module=SpawnRate001}
   &SpawnRate(SpawnGroup=2, SpawnRate=45, Use Spawn Probability=false) {module=SpawnRate002}
  ]
  particleSpawnScript [
   &VC_PaintedVariantsOverride() {enabled=false}
   /SetVariables(
    Particles.Base Color:LinearColor=&MakeCustomLinearColorFromBool(Bool=$User.bIsPainted, True Color=$User.PaintedColor, False Color=<rgba 0.2 0.8 0.05 1>)
    Particles.VisibilityTag=&SpawnGroup()
   ) {isSetParameters=true, module=SetVariables_#3}
   &InitializeParticle(
    Color Mode=::NewEnumerator0
    Lifetime Max=&VC_FloatBySpawnGroup(
     SpawnGroup or ExecIndex:NiagaraInt32=0
     CustomInt:NiagaraInt32=0
     MaximumSpawnGroups:NiagaraInt32=3
     Float SG 0:NiagaraFloat=1
     Float SG 1:NiagaraFloat=0.6
     Float SG 2:NiagaraFloat=0.8
     Float SG 3:NiagaraFloat=0
     Float SG 4:NiagaraFloat=0
     Float SG 5:NiagaraFloat=0
    )
    Lifetime Min:NiagaraFloat=&VC_FloatBySpawnGroup(
     SpawnGroup or ExecIndex:NiagaraInt32=0
     CustomInt:NiagaraInt32=0
     MaximumSpawnGroups:NiagaraInt32=3
     Float SG 0:NiagaraFloat=0.5
     Float SG 1:NiagaraFloat=0.5
     Float SG 2:NiagaraFloat=0.8
     Float SG 3:NiagaraFloat=0
     Float SG 4:NiagaraFloat=0
     Float SG 5:NiagaraFloat=0
    )
    Lifetime Mode=::NewEnumerator1
    Sprite Rotation Mode=::NewEnumerator2
    Sprite Size Mode=::NewEnumerator3
    Uniform Sprite Size=&VC_FloatBySpawnGroup(
     SpawnGroup or ExecIndex:NiagaraInt32=0
     CustomInt:NiagaraInt32=&SpawnGroup()
     MaximumSpawnGroups:NiagaraInt32=3
     Float SG 0:NiagaraFloat=60
     Float SG 1:NiagaraFloat=&RandomRangeFloat(Minimum=90, Maximum=150)
     Float SG 2:NiagaraFloat=&RandomRangeFloat(
      Minimum=&MakeCustomFloatFromBool(Bool=$User.bIsFrontendPreview, True Float=5, False Float=20)
      Maximum=&MakeCustomFloatFromBool(Bool=$User.bIsFrontendPreview, True Float=20, False Float=50)
     )
     Float SG 3:NiagaraFloat=0
     Float SG 4:NiagaraFloat=0
     Float SG 5:NiagaraFloat=0
    )
    UsePositionOffset=false
    Write Material Random=true
   )
   &PartitionParticles(Number Of Partitions=$Emitter.PortCount, Partition Mode:ENiagara_PartitionMode=::NewEnumerator5)
   &SpriteFacingAndAlignment(Facing Coordinate Space=::Local, WriteAlignment=false, WriteFacing=true)
   &SkeletalMeshLocation(
    Bone / Socket Index=$Particles.PartitionParticles.Partition
    Bone / Socket Sampling Mode=::NewEnumerator5
    MaskSpawnGroup=false
    Mesh Sampling Type=::NewEnumerator6
    Sampled Position Offset=&AddVectorToPosition(
     Position=&MakeVector(
      X=&VC_FloatBySpawnGroup(
       SpawnGroup or ExecIndex:NiagaraInt32=0
       CustomInt:NiagaraInt32=0
       MaximumSpawnGroups:NiagaraInt32=3
       Float SG 0:NiagaraFloat=-10
       Float SG 1:NiagaraFloat=-50
       Float SG 2:NiagaraFloat=-50
       Float SG 3:NiagaraFloat=0
       Float SG 4:NiagaraFloat=0
       Float SG 5:NiagaraFloat=0
      )
      Y=0
      Z=0
     )
     Vector=$User.DTCarOffset
    )
    Skeletal Mesh=$System.SkeletalMesh
   )
   &AddVelocity(Velocity=&MakeCustomPositionFromBool(
     Bool:NiagaraBool=$User.bIsFrontendPreview
     True Position:NiagaraPosition=&MakeVector(
      X=&VC_FloatBySpawnGroup(
       SpawnGroup or ExecIndex:NiagaraInt32=0
       CustomInt:NiagaraInt32=0
       MaximumSpawnGroups:NiagaraInt32=3
       Float SG 0:NiagaraFloat=-950
       Float SG 1:NiagaraFloat=-850
       Float SG 2:NiagaraFloat=-850
       Float SG 3:NiagaraFloat=0
       Float SG 4:NiagaraFloat=0
       Float SG 5:NiagaraFloat=0
      )
      Y=&MaskFloatBySpawnGroup(Masked Float=&RandomRangeFloat(Minimum=-80, Maximum=80), Passthrough Float=0, Spawn Group Mask=2)
      Z=&MaskFloatBySpawnGroup(Masked Float=&RandomRangeFloat(Minimum=-100, Maximum=100), Passthrough Float=0, Spawn Group Mask=2)
     )
     False Position:NiagaraPosition=<0 0 0>
    ))
   &Color(Color=&VC_Dynamic_ColorBySpawnGroup(
     Color SG 0:LinearColor=$Particles.Base Color
     MaximumSpawnGroups:NiagaraInt32=3
     Color SG 1:LinearColor=&MakeCustomLinearColorFromBool(Bool=$User.bIsPainted, True Color=$User.PaintedColor, False Color=<rgba 0.5 1 0.01 1>)
     Color SG 2:LinearColor=$Particles.Base Color
     Color SG 3:LinearColor=<rgba 1 1 1 1>
     Color SG 4:LinearColor=<rgba 1 1 1 1>
    ))
   &ShapeLocation(
    MaskSpawnGroup=false
    Override Seed=false
    Sphere Radius=&MaskFloatBySpawnGroup(Masked Float=20, Passthrough Float=0, Spawn Group Mask=2)
    UseOwnerScale=true
   )
  ]
  particleUpdateScript [
   &ParticleState()
   &ScaleSpriteSize(
    Scale Sprite Size Mode=::NewEnumerator0
    Uniform Scale Factor=&VC_FloatBySpawnGroup(
     SpawnGroup or ExecIndex:NiagaraInt32=0
     CustomInt:NiagaraInt32=0
     MaximumSpawnGroups:NiagaraInt32=3
     Float SG 0:NiagaraFloat=&FloatFromCurve(FloatCurve=DataInterface() {propertyValues {Curve {keys [{arriveTangent=2, interpMode=RCIM_Cubic, time=0.3, value=1.15}, <at 1 2 RCIM_Cubic>]}}})
     Float SG 1:NiagaraFloat=1
     Float SG 2:NiagaraFloat=1
     Float SG 3:NiagaraFloat=0
     Float SG 4:NiagaraFloat=0
     Float SG 5:NiagaraFloat=0
    )
   )
   &SpriteRotationRate(Rotation Rate=&MaskFloatBySpawnGroup(Masked Float=&RandomRangeFloat(Minimum=-0.2, Maximum=0.2), Passthrough Float=0, Spawn Group Mask=1))
   &ScaleVelocity(
    Coordinate Space:ENiagaraCoordinateSpace=::Simulation
    Velocity Scale=&VectorFromFloat(Value=&VC_FloatBySpawnGroup(
      SpawnGroup or ExecIndex:NiagaraInt32=0
      CustomInt:NiagaraInt32=0
      MaximumSpawnGroups:NiagaraInt32=3
      Float SG 0:NiagaraFloat=&MakeCustomFloatFromBool(
       Bool=$User.bIsFrontendPreview
       True Float=&FloatFromCurve(FloatCurve=DataInterface() {propertyValues {Curve {keys [<at 0 0.8>,<at 0.5 0.15>]}}})
       False Float=&FloatFromCurve(FloatCurve=DataInterface() {propertyValues {Curve {keys [<at 0 0.25>,<at 1 0>]}}})
      )
      Float SG 1:NiagaraFloat=&FloatFromCurve(FloatCurve=DataInterface() {propertyValues {Curve {keys [<at 0 1 RCIM_Cubic>,<at 1 1.25 RCIM_Cubic>]}}})
      Float SG 2:NiagaraFloat=0.6
      Float SG 3:NiagaraFloat=0
      Float SG 4:NiagaraFloat=0
      Float SG 5:NiagaraFloat=0
     ))
   )
   &SolveForcesAndVelocity(Clamp Velocity=false, Limit Acceleration=false)
   &ScaleColor(
    Scale Alpha=&VC_FloatBySpawnGroup(
     SpawnGroup or ExecIndex:NiagaraInt32=0
     CustomInt:NiagaraInt32=0
     MaximumSpawnGroups:NiagaraInt32=3
     Float SG 0:NiagaraFloat=&FloatFromCurve(FloatCurve=DataInterface() {propertyValues {Curve {keys [<at -0.0046 1.0113 RCIM_Cubic>,<at 1.0011 0.002 RCIM_Cubic>]}}})
     Float SG 1:NiagaraFloat=&FloatFromCurve(FloatCurve=DataInterface() {propertyValues {Curve {keys [<at 0 1 RCIM_Cubic>,<at 1.0011 0.002 RCIM_Cubic>]}}})
     Float SG 2:NiagaraFloat=&FloatFromCurve(FloatCurve=DataInterface() {propertyValues {Curve {keys [<at -0.0046 1.0113 RCIM_Cubic>,<at 1.0011 0.002 RCIM_Cubic>]}}})
     Float SG 3:NiagaraFloat=0
     Float SG 4:NiagaraFloat=0
     Float SG 5:NiagaraFloat=0
    )
    Scale RGB=&VectorFromFloat(Value=&VC_FloatBySpawnGroup(
      SpawnGroup or ExecIndex:NiagaraInt32=0
      CustomInt:NiagaraInt32=0
      MaximumSpawnGroups:NiagaraInt32=3
      Float SG 0:NiagaraFloat=&Multiply_Float(A=2, B=$User.PaintedScalar)
      Float SG 1:NiagaraFloat=$User.PaintedScalar
      Float SG 2:NiagaraFloat=&Multiply_Float(A=6, B=$User.PaintedScalar)
      Float SG 3:NiagaraFloat=0
      Float SG 4:NiagaraFloat=0
      Float SG 5:NiagaraFloat=0
     ))
    ScaleA=true
    ScaleRGB=true
   )
   &SubUVAnimation(
    End Frame Range Override=16
    Output Alternate Renderer Binding=true
    "Sprite Renderer "=DataInterface() {propertyValues {SpriteRenderer=&self:Sprites001_0.NiagaraSpriteRendererProperties_0}}
    UseEndFrame=true
    UseStartFrame=false
   )
   &JitterPosition(Jitter Amount:NiagaraFloat=&VC_FloatBySpawnGroup(
     SpawnGroup or ExecIndex:NiagaraInt32=0
     CustomInt:NiagaraInt32=0
     MaximumSpawnGroups:NiagaraInt32=3
     Float SG 0:NiagaraFloat=0
     Float SG 1:NiagaraFloat=&RandomRangeFloat(Minimum=-2, Maximum=2)
     Float SG 2:NiagaraFloat=&RandomRangeFloat(Minimum=-2, Maximum=2)
     Float SG 3:NiagaraFloat=0
     Float SG 4:NiagaraFloat=0
     Float SG 5:NiagaraFloat=0
    ))
  ]
  rendererClasses [&NiagaraSpriteRendererProperties]
  renderers [
   {rendererClass=&NiagaraSpriteRendererProperties}
   {rendererClass=&NiagaraSpriteRendererProperties, rendererIndex=1}
   {rendererClass=&NiagaraSpriteRendererProperties, rendererIndex=2}
  ]
 }
 Sprites002 {
  bEnabled=false
  simTarget=GPUComputeSim
  emitterSpawnScript [/SetVariables(Emitter.AcidDistortion:NiagaraFloat=&RandomRangeFloat(Minimum=0, Maximum=0.5)) {isSetParameters=true, module=SetVariables_#2}]
  emitterUpdateScript [
   &EmitterState()
   &VC_FilteredSocketCount()
   &SpawnRate(SpawnRate=20, Use Spawn Probability=false)
   &SpawnRate(
    SpawnGroup=1
    SpawnRate=&MakeCustomFloatFromBool(Bool=$User.bIsFrontendPreview, True Float=50, False Float=80)
    Use Spawn Probability=false
   ) {module=SpawnRate001}
   &SpawnRate(SpawnGroup=2, SpawnRate=45, Use Spawn Probability=false) {module=SpawnRate002}
  ]
  particleSpawnScript [
   &VC_PaintedVariantsOverride() {enabled=false}
   /SetVariables(
    Particles.Base Color:LinearColor=&MakeCustomLinearColorFromBool(Bool=$User.bIsPainted, True Color=$User.PaintedColor, False Color=<rgba 0.2 0.8 0.05 1>)
    Particles.VisibilityTag=&SpawnGroup()
   ) {isSetParameters=true, module=SetVariables_#3}
   &InitializeParticle(
    Color Mode=::NewEnumerator0
    Lifetime Max=&VC_FloatBySpawnGroup(
     SpawnGroup or ExecIndex:NiagaraInt32=0
     CustomInt:NiagaraInt32=0
     MaximumSpawnGroups:NiagaraInt32=3
     Float SG 0:NiagaraFloat=1
     Float SG 1:NiagaraFloat=0.6
     Float SG 2:NiagaraFloat=1.5
     Float SG 3:NiagaraFloat=0
     Float SG 4:NiagaraFloat=0
     Float SG 5:NiagaraFloat=0
    )
    Lifetime Min:NiagaraFloat=&VC_FloatBySpawnGroup(
     SpawnGroup or ExecIndex:NiagaraInt32=0
     CustomInt:NiagaraInt32=0
     MaximumSpawnGroups:NiagaraInt32=3
     Float SG 0:NiagaraFloat=0.5
     Float SG 1:NiagaraFloat=0.5
     Float SG 2:NiagaraFloat=1.5
     Float SG 3:NiagaraFloat=0
     Float SG 4:NiagaraFloat=0
     Float SG 5:NiagaraFloat=0
    )
    Lifetime Mode=::NewEnumerator1
    Sprite Rotation Mode=::NewEnumerator2
    Sprite Size Mode=::NewEnumerator3
    Uniform Sprite Size=&VC_FloatBySpawnGroup(
     SpawnGroup or ExecIndex:NiagaraInt32=0
     CustomInt:NiagaraInt32=&SpawnGroup()
     MaximumSpawnGroups:NiagaraInt32=3
     Float SG 0:NiagaraFloat=60
     Float SG 1:NiagaraFloat=&RandomRangeFloat(Minimum=90, Maximum=150)
     Float SG 2:NiagaraFloat=&RandomRangeFloat(
      Minimum=&MakeCustomFloatFromBool(Bool=$User.bIsFrontendPreview, True Float=3, False Float=20)
      Maximum=&MakeCustomFloatFromBool(Bool=$User.bIsFrontendPreview, True Float=6, False Float=50)
     )
     Float SG 3:NiagaraFloat=0
     Float SG 4:NiagaraFloat=0
     Float SG 5:NiagaraFloat=0
    )
    UsePositionOffset=false
    Write Material Random=true
   )
   &PartitionParticles(Number Of Partitions=$Emitter.PortCount, Partition Mode:ENiagara_PartitionMode=::NewEnumerator5)
   &SpriteFacingAndAlignment(Facing Coordinate Space=::Local, WriteAlignment=false, WriteFacing=true)
   &SkeletalMeshLocation(
    Bone / Socket Index=$Particles.PartitionParticles.Partition
    Bone / Socket Sampling Mode=::NewEnumerator5
    MaskSpawnGroup=false
    Mesh Sampling Type=::NewEnumerator6
    Sampled Position Offset=&AddVectorToPosition(
     Position=&MakeVector(
      X=&VC_FloatBySpawnGroup(
       SpawnGroup or ExecIndex:NiagaraInt32=0
       CustomInt:NiagaraInt32=0
       MaximumSpawnGroups:NiagaraInt32=3
       Float SG 0:NiagaraFloat=-10
       Float SG 1:NiagaraFloat=-50
       Float SG 2:NiagaraFloat=-50
       Float SG 3:NiagaraFloat=0
       Float SG 4:NiagaraFloat=0
       Float SG 5:NiagaraFloat=0
      )
      Y=0
      Z=0
     )
     Vector=$User.DTCarOffset
    )
    Skeletal Mesh=$System.SkeletalMesh
   )
   &AddVelocity(Velocity=&MakeCustomPositionFromBool(
     Bool:NiagaraBool=$User.bIsFrontendPreview
     True Position:NiagaraPosition=&MakeVector(
      X=&VC_FloatBySpawnGroup(
       SpawnGroup or ExecIndex:NiagaraInt32=0
       CustomInt:NiagaraInt32=0
       MaximumSpawnGroups:NiagaraInt32=3
       Float SG 0:NiagaraFloat=-950
       Float SG 1:NiagaraFloat=-850
       Float SG 2:NiagaraFloat=-850
       Float SG 3:NiagaraFloat=0
       Float SG 4:NiagaraFloat=0
       Float SG 5:NiagaraFloat=0
      )
      Y=&MaskFloatBySpawnGroup(Masked Float=&RandomRangeFloat(Minimum=0, Maximum=0), Passthrough Float=0, Spawn Group Mask=2)
      Z=&MaskFloatBySpawnGroup(Masked Float=&RandomRangeFloat(Minimum=0, Maximum=0), Passthrough Float=0, Spawn Group Mask=2)
     )
     False Position:NiagaraPosition=<0 0 0>
    ))
   &Color(Color=&VC_Dynamic_ColorBySpawnGroup(
     Color SG 0:LinearColor=$Particles.Base Color
     MaximumSpawnGroups:NiagaraInt32=3
     Color SG 1:LinearColor=&MakeCustomLinearColorFromBool(Bool=$User.bIsPainted, True Color=$User.PaintedColor, False Color=<rgba 0.5 1 0.01 1>)
     Color SG 2:LinearColor=$Particles.Base Color
     Color SG 3:LinearColor=<rgba 1 1 1 1>
     Color SG 4:LinearColor=<rgba 1 1 1 1>
    ))
   &ShapeLocation(
    MaskSpawnGroup=false
    Override Seed=false
    Sphere Radius=&MaskFloatBySpawnGroup(Masked Float=50, Passthrough Float=0, Spawn Group Mask=2)
    UseOwnerScale=true
   )
  ]
  particleUpdateScript [
   &ParticleState()
   &ScaleSpriteSize(
    Scale Sprite Size Mode=::NewEnumerator0
    Uniform Scale Factor=&VC_FloatBySpawnGroup(
     SpawnGroup or ExecIndex:NiagaraInt32=0
     CustomInt:NiagaraInt32=0
     MaximumSpawnGroups:NiagaraInt32=3
     Float SG 0:NiagaraFloat=&FloatFromCurve(FloatCurve=DataInterface() {propertyValues {Curve {keys [{arriveTangent=2, interpMode=RCIM_Cubic, time=0.3, value=1.15}, <at 1 2 RCIM_Cubic>]}}})
     Float SG 1:NiagaraFloat=1
     Float SG 2:NiagaraFloat=1
     Float SG 3:NiagaraFloat=0
     Float SG 4:NiagaraFloat=0
     Float SG 5:NiagaraFloat=0
    )
   )
   &SpriteRotationRate(Rotation Rate=&MaskFloatBySpawnGroup(Masked Float=&RandomRangeFloat(Minimum=-0.2, Maximum=0.2), Passthrough Float=0, Spawn Group Mask=1))
   &ScaleVelocity(
    Coordinate Space:ENiagaraCoordinateSpace=::Simulation
    Velocity Scale=&VectorFromFloat(Value=&VC_FloatBySpawnGroup(
      SpawnGroup or ExecIndex:NiagaraInt32=0
      CustomInt:NiagaraInt32=0
      MaximumSpawnGroups:NiagaraInt32=3
      Float SG 0:NiagaraFloat=&MakeCustomFloatFromBool(
       Bool=$User.bIsFrontendPreview
       True Float=&FloatFromCurve(FloatCurve=DataInterface() {propertyValues {Curve {keys [<at 0 0.8>,<at 0.5 0.15>]}}})
       False Float=&FloatFromCurve(FloatCurve=DataInterface() {propertyValues {Curve {keys [<at 0 0.25>,<at 1 0>]}}})
      )
      Float SG 1:NiagaraFloat=&FloatFromCurve(FloatCurve=DataInterface() {propertyValues {Curve {keys [<at 0 1 RCIM_Cubic>,<at 1 1.25 RCIM_Cubic>]}}})
      Float SG 2:NiagaraFloat=&FloatFromCurve(FloatCurve=DataInterface() {propertyValues {Curve {keys [<at 0 0.65 RCIM_Cubic>,<at 1 0.2 RCIM_Cubic>]}}})
      Float SG 3:NiagaraFloat=0
      Float SG 4:NiagaraFloat=0
      Float SG 5:NiagaraFloat=0
     ))
   )
   &CurlNoiseForce(
    Bias Noise=false
    Noise Frequency=&MaskFloatBySpawnGroup(Masked Float=10, Passthrough Float=0, Spawn Group Mask=2)
    Noise Strength=&MaskFloatBySpawnGroup(Masked Float=1500, Passthrough Float=0, Spawn Group Mask=2)
    Pan Noise=false
   ) {enabled=false}
   &SolveForcesAndVelocity(Clamp Velocity=false, Limit Acceleration=false)
   &ScaleColor(
    Scale Alpha=&VC_FloatBySpawnGroup(
     SpawnGroup or ExecIndex:NiagaraInt32=0
     CustomInt:NiagaraInt32=0
     MaximumSpawnGroups:NiagaraInt32=3
     Float SG 0:NiagaraFloat=&FloatFromCurve(FloatCurve=DataInterface() {propertyValues {Curve {keys [<at 0 1 RCIM_Cubic>,<at 1 0 RCIM_Cubic>]}}})
     Float SG 1:NiagaraFloat=&FloatFromCurve(FloatCurve=DataInterface() {propertyValues {Curve {keys [<at 0 1 RCIM_Cubic>,<at 1.0011 0.002 RCIM_Cubic>]}}})
     Float SG 2:NiagaraFloat=&FloatFromCurve(FloatCurve=DataInterface() {propertyValues {Curve {keys [<at -0.0046 1.0113 RCIM_Cubic>,<at 1.0011 0.002 RCIM_Cubic>]}}})
     Float SG 3:NiagaraFloat=0
     Float SG 4:NiagaraFloat=0
     Float SG 5:NiagaraFloat=0
    )
    Scale RGB=&VectorFromFloat(Value=&VC_FloatBySpawnGroup(
      SpawnGroup or ExecIndex:NiagaraInt32=0
      CustomInt:NiagaraInt32=0
      MaximumSpawnGroups:NiagaraInt32=3
      Float SG 0:NiagaraFloat=&Multiply_Float(A=2, B=$User.PaintedScalar)
      Float SG 1:NiagaraFloat=$User.PaintedScalar
      Float SG 2:NiagaraFloat=&Multiply_Float(A=6, B=$User.PaintedScalar)
      Float SG 3:NiagaraFloat=0
      Float SG 4:NiagaraFloat=0
      Float SG 5:NiagaraFloat=0
     ))
    ScaleA=true
    ScaleRGB=true
   )
   &SubUVAnimation(
    End Frame Range Override=16
    Output Alternate Renderer Binding=true
    "Sprite Renderer "=DataInterface() {propertyValues {SpriteRenderer=&self:Sprites002_0.NiagaraSpriteRendererProperties_0}}
    UseEndFrame=true
    UseStartFrame=false
   )
   &JitterPosition(Jitter Amount:NiagaraFloat=&VC_FloatBySpawnGroup(
     SpawnGroup or ExecIndex:NiagaraInt32=0
     CustomInt:NiagaraInt32=0
     MaximumSpawnGroups:NiagaraInt32=3
     Float SG 0:NiagaraFloat=0
     Float SG 1:NiagaraFloat=&RandomRangeFloat(Minimum=-2, Maximum=2)
     Float SG 2:NiagaraFloat=&RandomRangeFloat(Minimum=-2, Maximum=2)
     Float SG 3:NiagaraFloat=0
     Float SG 4:NiagaraFloat=0
     Float SG 5:NiagaraFloat=0
    ))
  ]
  rendererClasses [&NiagaraSpriteRendererProperties]
  renderers [
   {rendererClass=&NiagaraSpriteRendererProperties}
   {rendererClass=&NiagaraSpriteRendererProperties, rendererIndex=1}
   {rendererClass=&NiagaraSpriteRendererProperties, rendererIndex=2}
  ]
 }
 Trail {
  bEnabled=false
  emitterUpdateScript [
   &EmitterState()
   &VC_FilteredSocketCount()
   &VC_RateQualityHandler(Low Spawn Rate=40, Medium Spawn Rate=45, Use Custom Rate?=true)
   &VC_SpawnRibbon_v1(Number of Ribbons=$Emitter.PortCount, Spawn Time=-1, SpawnRate=$Emitter.Spawn Rate)
  ]
  particleSpawnScript [
   &VC_PaintedVariantsOverride() {enabled=false}
   &VC_InitializeRibbon_v1(
    Custom Offset=true
    Enable Skeletal Position=true
    Local Vector Offset=&AddVectorToPosition(Position=<30 0 0>, Vector=$User.DTCarOffset)
    Offset Amount=1
    Ribbon Width=60
   )
   &AddVelocity(
    Velocity=&MakeCustomVectorFromBool(Bool=$User.bIsFrontendPreview, True Vector=<-1000 0 0>, False Vector=<-1500 0 0>)
    Velocity Speed Scale=&FloatFromCurve(
     FloatCurve=DataInterface() {propertyValues {Curve {keys [{interpMode=RCIM_Cubic, tangentMode=RCTM_User, time=0.5, value=1}, {interpMode=RCIM_Cubic, tangentMode=RCTM_User, time=1, value=0}]}}}
     Scale Curve=2
    )
   ) {module=AddVelocity001}
   &Color(Color=&MakeCustomLinearColorFromBool(Bool=$User.bIsPainted, True Color=$User.PaintedColor, False Color=<rgba 0.2 0.8 0.05 1>))
   /SetVariables(Particles.RibbonWidth2=45) {isSetParameters=true, module=SetVariables_#4}
   /VehicleCosmetics/FX/Niagara/Modules/MergeEmitters/VC_ExtraBindingColors(
    ExtraColors=1
    R1 Color:LinearColor=&MakeCustomLinearColorFromBool(Bool=$User.bIsPainted, True Color=$User.PaintedColor, False Color=<rgba 0.2 0.8 0.05 1>)
   )
  ]
  particleUpdateScript [
   &ParticleState()
   &InheritSourceMovement(
    Applied Position Scale=&Multiply_VectorByFloat(
     Vector=<1 1 1>
     Float=&FloatFromCurve(FloatCurve=DataInterface() {propertyValues {Curve {keys [{interpMode=RCIM_Cubic, tangentMode=RCTM_Break, time=0.3, value=1}, {interpMode=RCIM_Cubic, tangentMode=RCTM_Break, time=1, value=0}]}}})
    )
    ApplyForce=false
    ApplyPosition=true
    ApplyPositionFalloff=false
   )
   &SolveForcesAndVelocity(Clamp Velocity=false, Limit Acceleration=false)
   &VC_UpdateRibbon_v1(Enable Skeletal Position=true)
   &ScaleColor(
    Scale Alpha=&FloatFromCurve(FloatCurve=DataInterface() {propertyValues {Curve {keys [
         <at 0 0 RCIM_Cubic>
         {arriveTangent=1.25, leaveTangent=1.25, time=0.3, value=1}
         {arriveTangent=-1.25, leaveTangent=-1.25, time=0.8, value=1}
         <at 1 0 RCIM_Cubic>
        ]}}})
    Scale RGB=&VectorFromFloat(Value=&FloatFromCurve(FloatCurve=DataInterface() {propertyValues {Curve {keys [<at 0 1 RCIM_Cubic>,<at 1 0.25 RCIM_Cubic>]}}}))
    ScaleA=true
    ScaleRGB=true
   ) {enabled=false}
   /SetVariables(Particles.SecondTrailColor:LinearColor=&VC_DynamicScaleColor(
     Color to Scale:LinearColor=$Particles.Initial.R1 Color
     RGB Scale:Vector3f=<1 1 1>
     Alpha Scale:NiagaraFloat=&FloatFromCurve(FloatCurve=DataInterface() {propertyValues {Curve {keys [
          {interpMode=RCIM_Cubic, tangentMode=RCTM_Break, time=0, value=0}
          {arriveTangent=1.25, leaveTangent=1.25, time=0.3, value=1}
          {arriveTangent=-1.4286, interpMode=RCIM_Cubic, leaveTangent=-1.4286, time=0.8, value=1}
          <at 1 0 RCIM_Cubic>
         ]}}})
    )) {isSetParameters=true, module=SetVariables_#5}
   &DynamicMaterialParameters(
    Index 0 Param 1=&FloatFromCurve(FloatCurve=DataInterface() {propertyValues {Curve {keys [<at 0 -0.1 RCIM_Cubic>,<at 1 -0.5 RCIM_Cubic>]}}})
    Index 0 Param 2=&RandomRangeFloat(Minimum=0, Maximum=1.5)
    Module.Index0Param0Write=true
    Module.Index0Param1Write=true
    Module.Index0Param2Write=true
    Module.Index0Param3Write=true
   )
  ]
  rendererClasses [&NiagaraRibbonRendererProperties]
  renderers [
   {rendererClass=&NiagaraRibbonRendererProperties}
   {rendererClass=&NiagaraRibbonRendererProperties, rendererIndex=1}
  ]
 }
 Trail001 {
  emitterUpdateScript [
   &EmitterState()
   &VC_FilteredSocketCount()
   &VC_RateQualityHandler(Low Spawn Rate=40, Medium Spawn Rate=45, Use Custom Rate?=true)
   &VC_SpawnRibbon_v1(Number of Ribbons=$Emitter.PortCount, Spawn Time=-1, SpawnRate=$Emitter.Spawn Rate)
  ]
  particleSpawnScript [
   &VC_PaintedVariantsOverride() {enabled=false}
   &VC_InitializeRibbon_v1(
    Custom Offset=true
    Enable Skeletal Position=true
    Local Vector Offset=&AddVectorToPosition(Position=<30 0 0>, Vector=$User.DTCarOffset)
    Offset Amount=1
    Ribbon Width=50
   )
   &AddVelocity(
    Velocity=&MakeCustomVectorFromBool(Bool=$User.bIsFrontendPreview, True Vector=<-1800 0 0>, False Vector=<-2000 0 0>)
    Velocity Speed Scale=&FloatFromCurve(
     FloatCurve=DataInterface() {propertyValues {Curve {keys [{interpMode=RCIM_Cubic, tangentMode=RCTM_User, time=0.5, value=1}, {interpMode=RCIM_Cubic, tangentMode=RCTM_User, time=1, value=0}]}}}
     Scale Curve=2
    )
   ) {module=AddVelocity001}
   &Color(Color=&MakeCustomLinearColorFromBool(Bool=$User.bIsPainted, True Color=$User.PaintedColor, False Color=<rgba 0.32 0.81 0.09 1>))
   /SetVariables(Particles.RibbonWidth2=50) {isSetParameters=true, module=SetVariables_#4}
   &VC_ExtraBindingColors(
    ExtraColors=2
    R1 Color:LinearColor=&MakeCustomLinearColorFromBool(Bool=$User.bIsPainted, True Color=$User.PaintedColor, False Color=<rgba 0.2 0.8 0.05 1>)
    R2 Color:LinearColor=&MakeCustomLinearColorFromBool(Bool=$User.bIsPainted, True Color=$User.PaintedColor, False Color=<rgba 0.55 1 0.13 1>)
   )
  ]
  particleUpdateScript [
   &ParticleState()
   &InheritSourceMovement(
    Applied Position Scale=&Multiply_VectorByFloat(
     Vector=<1 1 1>
     Float=&FloatFromCurve(FloatCurve=DataInterface() {propertyValues {Curve {keys [{interpMode=RCIM_Cubic, tangentMode=RCTM_Break, time=0.3, value=1}, {interpMode=RCIM_Cubic, tangentMode=RCTM_Break, time=1, value=0}]}}})
    )
    ApplyForce=false
    ApplyPosition=true
    ApplyPositionFalloff=false
   )
   &SolveForcesAndVelocity(Clamp Velocity=false, Limit Acceleration=false)
   &VC_UpdateRibbon_v1(Enable Skeletal Position=true)
   &ScaleRibbonWidth(Ribbon Width Scale=&FloatFromCurve(FloatCurve=DataInterface() {propertyValues {Curve {keys [<at 0 1 RCIM_Cubic>,<at 0.5 0.5 RCIM_Cubic>]}}}))
   &ScaleColor(
    Scale Alpha=&FloatFromCurve(FloatCurve=DataInterface() {propertyValues {Curve {keys [
         <at 0 0 RCIM_Cubic>
         {arriveTangent=1.25, leaveTangent=1.25, time=0.3, value=1}
         {arriveTangent=-1.25, leaveTangent=-1.25, time=0.8, value=1}
         <at 1 0 RCIM_Cubic>
        ]}}})
    Scale RGB=&VectorFromFloat(Value=&FloatFromCurve(FloatCurve=DataInterface() {propertyValues {Curve {keys [<at 0 1.25 RCIM_Cubic>,<at 1 0.25 RCIM_Cubic>]}}}))
    ScaleA=true
    ScaleRGB=true
   )
   /SetVariables(
    Particles.SecondColorScale:LinearColor=&Multiply_LinearColorByFloat(Linear Color=$Particles.R2 Color, Float=1.2)
    Particles.SecondTrailColor:LinearColor=&VC_DynamicScaleColor(
     Color to Scale:LinearColor=$Particles.Initial.R1 Color
     RGB Scale:Vector3f=<1 1 1>
     Alpha Scale:NiagaraFloat=&FloatFromCurve(FloatCurve=DataInterface() {propertyValues {Curve {keys [
          {interpMode=RCIM_Cubic, tangentMode=RCTM_Break, time=0, value=0}
          {arriveTangent=1.25, leaveTangent=1.25, time=0.3, value=1}
          {arriveTangent=-1.4286, interpMode=RCIM_Cubic, leaveTangent=-1.4286, time=0.8, value=1}
          <at 1 0 RCIM_Cubic>
         ]}}})
    )
    Particles.SecondTrailSize:NiagaraFloat=&Multiply_Float(
     A=$Particles.RibbonWidth2
     B=&FloatFromCurve(FloatCurve=DataInterface() {propertyValues {Curve {keys [<at 0 1 RCIM_Cubic>,<at 0.5 0.2 RCIM_Cubic>]}}})
    )
   ) {isSetParameters=true, module=SetVariables_#5}
   &DynamicMaterialParameters(
    Index 0 Param 1=&FloatFromCurve(FloatCurve=DataInterface() {propertyValues {Curve {keys [<at 0 -0.1 RCIM_Cubic>,<at 1 -0.5 RCIM_Cubic>]}}})
    Index 0 Param 2=&RandomRangeFloat(Minimum=1, Maximum=1.5)
    Module.Index0Param0Write=true
    Module.Index0Param1Write=true
    Module.Index0Param2Write=false
    Module.Index0Param3Write=false
   )
  ]
  rendererClasses [&NiagaraRibbonRendererProperties]
  renderers [
   {rendererClass=&NiagaraRibbonRendererProperties}
   {rendererClass=&NiagaraRibbonRendererProperties, rendererIndex=1}
  ]
 }
 Trail002 {
  bEnabled=false
  emitterUpdateScript [
   &EmitterState()
   &VC_FilteredSocketCount()
   &VC_RateQualityHandler(Low Spawn Rate=40, Medium Spawn Rate=45, Use Custom Rate?=true)
   &VC_SpawnRibbon_v1(Number of Ribbons=$Emitter.PortCount, Spawn Time=-1, SpawnRate=$Emitter.Spawn Rate)
  ]
  particleSpawnScript [
   &VC_PaintedVariantsOverride() {enabled=false}
   &VC_InitializeRibbon_v1(
    Custom Offset=true
    Enable Skeletal Position=true
    Local Vector Offset=&AddVectorToPosition(Position=<30 0 0>, Vector=$User.DTCarOffset)
    Offset Amount=1
    Ribbon Width=50
   )
   &AddVelocity(
    Velocity=&MakeCustomVectorFromBool(Bool=$User.bIsFrontendPreview, True Vector=<-1800 0 0>, False Vector=<-1500 0 0>)
    Velocity Speed Scale=&FloatFromCurve(
     FloatCurve=DataInterface() {propertyValues {Curve {keys [{interpMode=RCIM_Cubic, tangentMode=RCTM_User, time=0.5, value=1}, {interpMode=RCIM_Cubic, tangentMode=RCTM_User, time=1, value=0}]}}}
     Scale Curve=2
    )
   ) {module=AddVelocity001}
   &Color(Color=&MakeCustomLinearColorFromBool(Bool=$User.bIsPainted, True Color=$User.PaintedColor, False Color=<rgba 0.32 0.81 0.09 1>))
   /SetVariables(Particles.RibbonWidth2=50) {isSetParameters=true, module=SetVariables_#4}
   &VC_ExtraBindingColors(
    ExtraColors=2
    R1 Color:LinearColor=&MakeCustomLinearColorFromBool(Bool=$User.bIsPainted, True Color=$User.PaintedColor, False Color=<rgba 0.2 0.8 0.05 1>)
    R2 Color:LinearColor=&MakeCustomLinearColorFromBool(Bool=$User.bIsPainted, True Color=$User.PaintedColor, False Color=<rgba 0.55 1 0.13 1>)
   )
  ]
  particleUpdateScript [
   &ParticleState()
   &InheritSourceMovement(
    Applied Position Scale=&Multiply_VectorByFloat(
     Vector=<1 1 1>
     Float=&FloatFromCurve(FloatCurve=DataInterface() {propertyValues {Curve {keys [{interpMode=RCIM_Cubic, tangentMode=RCTM_Break, time=0.3, value=1}, {interpMode=RCIM_Cubic, tangentMode=RCTM_Break, time=1, value=0}]}}})
    )
    ApplyForce=false
    ApplyPosition=true
    ApplyPositionFalloff=false
   )
   &SolveForcesAndVelocity(Clamp Velocity=false, Limit Acceleration=false)
   &VC_UpdateRibbon_v1(Enable Skeletal Position=true)
   &ScaleRibbonWidth(Ribbon Width Scale=&FloatFromCurve(FloatCurve=DataInterface() {propertyValues {Curve {keys [<at 0 1 RCIM_Cubic>,<at 0.5 0.5 RCIM_Cubic>]}}}))
   &ScaleColor(
    Scale Alpha=&FloatFromCurve(FloatCurve=DataInterface() {propertyValues {Curve {keys [
         <at 0 0 RCIM_Cubic>
         {arriveTangent=1.25, leaveTangent=1.25, time=0.3, value=1}
         {arriveTangent=-1.25, leaveTangent=-1.25, time=0.8, value=1}
         <at 1 0 RCIM_Cubic>
        ]}}})
    Scale RGB=&VectorFromFloat(Value=&FloatFromCurve(FloatCurve=DataInterface() {propertyValues {Curve {keys [<at 0 1.25 RCIM_Cubic>,<at 1 0.25 RCIM_Cubic>]}}}))
    ScaleA=true
    ScaleRGB=true
   )
   /SetVariables(
    Particles.SecondColorScale:LinearColor=&Multiply_LinearColorByFloat(Linear Color=$Particles.R2 Color, Float=1.2)
    Particles.SecondTrailColor:LinearColor=&VC_DynamicScaleColor(
     Color to Scale:LinearColor=$Particles.Initial.R1 Color
     RGB Scale:Vector3f=<1 1 1>
     Alpha Scale:NiagaraFloat=&FloatFromCurve(FloatCurve=DataInterface() {propertyValues {Curve {keys [
          {interpMode=RCIM_Cubic, tangentMode=RCTM_Break, time=0, value=0}
          {arriveTangent=1.25, leaveTangent=1.25, time=0.3, value=1}
          {arriveTangent=-1.4286, interpMode=RCIM_Cubic, leaveTangent=-1.4286, time=0.8, value=1}
          <at 1 0 RCIM_Cubic>
         ]}}})
    )
    Particles.SecondTrailSize:NiagaraFloat=&Multiply_Float(
     A=$Particles.RibbonWidth2
     B=&FloatFromCurve(FloatCurve=DataInterface() {propertyValues {Curve {keys [<at 0 1 RCIM_Cubic>,<at 0.5 0.2 RCIM_Cubic>]}}})
    )
   ) {isSetParameters=true, module=SetVariables_#5}
   &DynamicMaterialParameters(
    Index 0 Param 1=&FloatFromCurve(FloatCurve=DataInterface() {propertyValues {Curve {keys [<at 0 -0.1 RCIM_Cubic>,<at 1 -0.5 RCIM_Cubic>]}}})
    Index 0 Param 2=&RandomRangeFloat(Minimum=1, Maximum=1.5)
    Module.Index0Param0Write=true
    Module.Index0Param1Write=true
    Module.Index0Param2Write=false
    Module.Index0Param3Write=false
   )
  ]
  rendererClasses [&NiagaraRibbonRendererProperties]
  renderers [
   {rendererClass=&NiagaraRibbonRendererProperties}
   {rendererClass=&NiagaraRibbonRendererProperties, rendererIndex=1}
  ]
 }
}
summary {
 systemName=NS_Boost_AlienLiquid_Optimized
 emitters [
  {bEnabled=false, emitterName=Trail, rendererClasses [&NiagaraRibbonRendererProperties]}
  {bEnabled=false, emitterName=Sprites001, simTarget=GPUComputeSim, rendererClasses [&NiagaraSpriteRendererProperties]}
  {emitterName=Trail001, rendererClasses [&NiagaraRibbonRendererProperties]}
  {emitterName=Sprites, simTarget=GPUComputeSim, rendererClasses [&NiagaraSpriteRendererProperties]}
  {bEnabled=false, emitterName=Trail002, rendererClasses [&NiagaraRibbonRendererProperties]}
  {bEnabled=false, emitterName=Sprites002, simTarget=GPUComputeSim, rendererClasses [&NiagaraSpriteRendererProperties]}
 ]
 userVariables [
  User.bIsFrontendPreview {type=NiagaraBool}
  User.bIsPainted {type=NiagaraBool}
  User.ColorVariantID {type=NiagaraInt32}
  User.DTCarOffset {type=Vector3f, description="", defaultValue {struct=Vector3f, value=<0 0 0>}}
  User.PaintedColor {type=LinearColor, description="", defaultValue {struct=LinearColor, value=<rgba 1 1 1 1>}}
  User.PaintedScalar {type=NiagaraFloat, defaultValue=1, description=""}
  User.PortCount {type=NiagaraInt32, defaultValue=2}
 ]
}
