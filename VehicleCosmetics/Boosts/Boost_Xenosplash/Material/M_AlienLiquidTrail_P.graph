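# Unlit translucent Niagara ribbon material for the alien liquid trail of the Xenosplash boost.
# Two panning samples of a cloud noise texture distort UV gradients into liquid masks, which blend particle color toward a brighter version for emissive.
# A Step node cuts the eroding masks into a hard-edged shape, multiplied by a ramp along U.
# Time-driven sine and cosine waves build a transformed wobble vector that scales along the second UV set.
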Material /VehicleCosmetics/Boosts/Boost_Xenosplash/Material/M_AlienLiquidTrail_P
graph {
 properties {
  bUsedWithNiagaraRibbons=true
  blendMode=BLEND_Translucent
  pixelDepthOffsetMode=PDOM_Legacy
  refractionMethod=RM_PixelNormalOffset
  shadingModel=MSM_Unlit
  usedShadingModels=Unlit
 }
 nodes [
  n1 = TextureSample(UVs=Panner(Coordinate=TextureCoordinate() {coordinateIndex=1, uTiling=2, vTiling=0.25}) {speedX=-1, speedY=0.15}) {texture=/DealershipSharedLibrary/FX/Textures/Noise/T_VCVFX_CloudsSoft_RGB}
  n2 = TextureSample(UVs=Panner(Coordinate=TextureCoordinate() {coordinateIndex=1, vTiling=0.25}) {speedX=-0.8, speedY=-0.1}) {texture=&T_VCVFX_CloudsSoft_RGB}
  n3 = AppendVector(A=@n1.G, B=@n2.R)
  AppendVector(A=Multiply(A=Constant() {r=0.5}, B=@n3), B=Add(A=@n1.B, B=@n2.B))
  n4 = ParticleColor()
  n5 = Multiply(A=@n4.RGB, B=Constant() {r=3})
  n6 = TextureCoordinate()
  n7 = ComponentMask(None=@n6) {r=true}
  n8 = ComponentMask(None=@n6) {g=true}
  n9 = Multiply(A=@n8, B=OneMinus(None=@n8))
  n10 = Constant() {r=5}
  n11 = ComponentMask(None=Multiply(A=@n3, B=Multiply(A=@n9, B=@n10))) {g=true}
  n12 = OneMinus(None=Multiply(A=@n7, B=@n11))
  n13 = Multiply(A=@n3, B=Constant() {r=0.15})
  n14 = ComponentMask(None=Add(A=Add(A=@n6, B=@n13), B=Constant2Vector() {g=-0.15})) {g=true}
  n15 = Multiply(A=Power(Base=OneMinus(None=@n14), Exp=Constant() {r=2.25}), B=Constant() {r=3})
  n16 = Saturate(None=ConstantBiasScale(None=Multiply(A=@n14, B=@n15)) {bias=-0.1, scale=20})
  n17 = Saturate(None=ConstantBiasScale(None=Multiply(A=@n12, B=@n16)) {bias=-0.9, scale=5})
  MaterialFunctionCall(Emissive=Clamp(None=LinearInterpolate(A=@n4.RGB, B=@n5, Alpha=@n17)) {maxDefault=10}) {materialFunction=&MF_FX_EmissiveAdjustments}
  n18 = Multiply(A=Constant() {r=3}, B=OneMinus(None=Time()))
  n19 = ComponentMask(None=Multiply(A=TextureCoordinate() {coordinateIndex=1, uTiling=2}, B=Constant() {r=1})) {r=true}
  n20 = Add(A=@n18, B=@n19)
  n21 = Add(A=Cosine(None=@n20), B=Cosine(None=@n20) {period=2})
  n22 = Add(A=@n19, B=@n18)
  n23 = AppendVector(A=@n21, B=Add(A=Sine(None=@n22), B=Sine(None=@n22) {period=2}))
  Multiply(A=Transform(None=AppendVector(A=@n23, B=@n21)), B=ComponentMask(None=TextureCoordinate() {coordinateIndex=1}) {r=true})
  n24 = Power(Base=ComponentMask(None=TextureCoordinate()) {r=true}) {constExponent=1.09}
  n25 = Multiply(A=Constant() {r=2}, B=OneMinus(None=@n7))
  n26 = Multiply(A=@n11, B=@n25)
  n27 = Multiply(A=Saturate(None=ConstantBiasScale(None=@n26) {bias=-0.1, scale=5}), B=@n16)
  Multiply(A=Saturate(None=Multiply(A=@n24) {constB=24.87}), B=Step(X=Multiply(A=@n25, B=@n27)) {constY=0.1})
 ]
}
